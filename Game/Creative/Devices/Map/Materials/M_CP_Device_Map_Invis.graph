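# Masked, Default Lit material for the Creative map device, named as its invisible variant.
# The graph holds no expressions, so no textures, parameters or logic drive the surface.
# It sets legacy pixel depth offset mode and shows no opacity mask input, so the file itself does not show how invisibility is achieved.
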Material /Game/Creative/Devices/Map/Materials/M_CP_Device_Map_Invis
graph {expressions {}, properties {blendMode=BLEND_Masked, pixelDepthOffsetMode=PDOM_Legacy, usedShadingModels=Default Lit}}
