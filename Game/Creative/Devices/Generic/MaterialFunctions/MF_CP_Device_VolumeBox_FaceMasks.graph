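# Material function, exposed to the library, that outputs a face mask for Creative device volume boxes from a Mask ID: one of six sides or all sides.
# Single-face masks derive from world-space vertex normal components, using negation and saturation for opposite faces.
# Chained If nodes select the mask by ID.
# A static switch optionally inverts it, and CheapContrast cleans up artifacts when combining masks.
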MaterialFunction /Game/Creative/Devices/Generic/MaterialFunctions/MF_CP_Device_VolumeBox_FaceMasks
graph {
 properties {bExposeToLibrary=true, description=Mask Volumes Based on Faces}
 nodes [
  n1 = NamedRerouteUsage()
  n2 = NamedRerouteUsage()
  n3 = Constant() {r=3}
  n4 = NamedRerouteUsage()
  n5 = NamedRerouteUsage()
  n6 = FunctionInput(Preview=Constant()) {
   bUsePreviewValueAsDefault=true
   desc="0 = Top, 1 = Bottom, 2 = Front, 3 = Back, 4 = Left, 5 = Right, 6 = All Sides"
   description="0 = Top, 1 = Bottom, 2 = Front, 3 = Back, 4 = Left, 5 = Right, 6 = All Sides"
   inputName=Mask ID
   inputType=FunctionInput_Scalar
   sortPriority=1
  }
  n7 = NamedRerouteUsage()
  n8 = If(
   A=NamedRerouteUsage()
   B=Constant() {r=2}
   "A > B"=@n5
   "A == B"=@n5
   "A < B"=If(
    A=NamedRerouteDeclaration(None=@n6) {name=Input ID, nodeColor=<rgba 1 0.8266 0 1>}
    B=Constant() {r=1}
    "A > B"=@n7
    "A == B"=@n7
    "A < B"=NamedRerouteUsage()
   )
  )
  n9 = If(
   A=NamedRerouteUsage()
   B=Constant() {r=5}
   "A > B"=NamedRerouteUsage()
   "A == B"=NamedRerouteUsage()
   "A < B"=If(A=NamedRerouteUsage(), B=Constant() {r=4}, "A > B"=@n1, "A == B"=@n1, "A < B"=If(A=@n2, B=@n3, "A > B"=@n4, "A == B"=@n4, "A < B"=@n8))
  )
  n10 = FunctionInput(Preview=StaticBoolParameter() {parameterName=Invert}) {bUsePreviewValueAsDefault=true, desc="Default = False", inputName=Invert Mask, inputType=FunctionInput_StaticBool, sortPriority=2}
  n11 = MaterialFunctionCall(In=StaticSwitch(True=OneMinus(None=@n9), False=@n9, Value=@n10), Contrast=Constant() {r=1}) {desc=Fixes Artifacts when Combining Masks, materialFunction=&CheapContrast}
  FunctionOutput(None=@n11.Result)
  n12 = ComponentMask(None=VertexNormalWS()) {b=true}
  n13 = Saturate(None=Multiply(A=NamedRerouteDeclaration(None=@n12) {name=Top, nodeColor=<rgba 0 0.3906 1 1>}) {constB=-1})
  NamedRerouteDeclaration(None=@n13) {name=Bottom, nodeColor=<rgba 0 1 0.0625 1>}
  n14 = ComponentMask(None=VertexNormalWS()) {g=true}
  n15 = Saturate(None=Multiply(A=NamedRerouteDeclaration(None=@n14) {name=Back, nodeColor=<rgba 0 0.0156 1 1>}) {constB=-1})
  NamedRerouteDeclaration(None=@n15) {name=Front, nodeColor=<rgba 1 0.0234 0 1>}
  n16 = ComponentMask(None=VertexNormalWS()) {r=true}
  n17 = Saturate(None=Multiply(A=NamedRerouteDeclaration(None=@n16) {name=Left, nodeColor=<rgba 0 0.8594 1 1>}) {constB=-1})
  NamedRerouteDeclaration(None=@n17) {name=Right, nodeColor=<rgba 1 0 0.8711 1>}
  n18 = NamedRerouteUsage()
  n19 = NamedRerouteUsage()
  NamedRerouteDeclaration(None=OneMinus(None=LinearInterpolate(A=@n18, Alpha=@n19))) {name=All Sides, nodeColor=<rgba 0 0.5878 1 1>}
  Reroute(None=@n13)
  Reroute()
  Reroute(None=@n12)
  Reroute(None=@n16)
  Reroute(None=@n14)
  Reroute(None=@n15)
  Reroute(None=@n17)
  Reroute(None=@n6)
  Reroute()
 ]
}
