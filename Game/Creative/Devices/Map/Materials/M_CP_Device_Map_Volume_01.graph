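# Translucent two-sided unlit material for the Creative Map device volume, with an Is Loaded blend for level-loader state.
# It layers animated world-space noise blocks, a bottom glow, and optional grid and arrow icons placed by UV-channel face masks.
# An optional hologram adds noisy Fresnel, scanlines, flicker, pulse and vertex offset.
# Backface and camera-distance terms shape opacity, and the result feeds MF_CreativeMoveTool.
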Material /Game/Creative/Devices/Map/Materials/M_CP_Device_Map_Volume_01
graph {
 properties {
  allowTranslucentCustomDepthWrites=true
  bEnableResponsiveAA=true
  bUseMaterialAttributes=true
  blendMode=BLEND_Translucent
  pixelDepthOffsetMode=PDOM_Legacy
  shadingModel=MSM_Unlit
  twoSided=true
  usedShadingModels=Unlit
 }
 nodes [
  n1 = Constant()
  n2 = NamedRerouteUsage()
  If(
   A=MaterialFunctionCall(Float2=TextureCoordinate()) {materialFunction=&BreakOutFloat2Components}.R
   B=Divide(B=ScalarParameter() {defaultValue=1, group=Grid, parameterName=Height, sortPriority=3})
   "A > B"=@n1
   "A == B"=@n1
   "A < B"=@n2
  )
  n3 = NamedRerouteUsage() {desc=Alpha - Grid Only}
  n4 = NamedRerouteUsage()
  n5 = NamedRerouteUsage()
  n6 = NamedRerouteDeclaration(None=StaticSwitchParameter(True=LinearInterpolate(A=@n5, B=@n4), False=@n5) {defaultValue=true, parameterName=IsHologram, sortPriority=0}) {name=Level Loader Glow Only - Color, nodeColor=<rgba 0 0.5078 1 1>}
  n7 = StaticSwitchParameter(
   True=LinearInterpolate(A=@n6, B=VectorParameter() {group=Grid, parameterName=Grid Color, sortPriority=1, defaultValue=<rgba 1 1 1 1>}.RGB, Alpha=@n3) {constAlpha=1}
   False=@n6
  ) {desc=Warning: Grid scaling has not been updated to match Icon/Arrow scaling, group=Grid, parameterName=Use Grid, sortPriority=0}
  n8 = StaticSwitchParameter(
   True=LinearInterpolate(A=@n7, B=VectorParameter() {group=Icon, parameterName=Icon Color, defaultValue=<rgba 1 1 1 1>}.RGB, Alpha=NamedRerouteUsage())
   False=@n7
  ) {defaultValue=true, group=Icon, parameterName=Use Icon, sortPriority=1}
  n9 = NamedRerouteUsage()
  n10 = ScalarParameter() {group=Level Loader, parameterName=Is Loaded, sliderMax=1, sortPriority=10}
  n11 = NamedRerouteUsage()
  n12 = Add(A=@n11, B=NamedRerouteUsage())
  n13 = Constant2Vector()
  n14 = TwoSidedSign()
  n15 = StaticSwitchParameter(True=Add(A=@n3, B=@n11), False=@n11) {desc=Warning: Grid scaling has not been updated to match Icon/Arrow scaling, group=Grid, parameterName=Use Grid, sortPriority=0}
  n16 = StaticSwitchParameter(True=LinearInterpolate(A=NamedRerouteUsage(), B=AppendVector(A=@n12, B=@n13), Alpha=Saturate(None=@n14)) {constB=0}, False=@n15) {defaultValue=true, parameterName=IsHologram, sortPriority=0}
  n17 = ScalarParameter() {defaultValue=1, group=Distance Fade, parameterName=Enable Distance Fade, sortPriority=1}
  n18 = LinearInterpolate(A=@n16, B=Multiply(A=@n16, B=NamedRerouteUsage()), Alpha=@n17) {constA=1}
  n19 = Multiply(A=NamedRerouteUsage(), B=NamedRerouteUsage())
  n20 = TwoSidedSign()
  n21 = StaticSwitchParameter(
   True=LinearInterpolate(A=@n18, B=LinearInterpolate(B=@n19, Alpha=@n17) {constA=1}, Alpha=Saturate(None=@n20)) {constA=1, constB=0}
   False=@n18
  ) {desc=Warning: Grid scaling has not been updated to match Icon/Arrow scaling, group=Grid, parameterName=Use Grid, sortPriority=0}
  n22 = StaticSwitchParameter(True=@n21, False=@n18) {desc=Does not work with Hologram, parameterName=Fill Backface Only}
  n23 = Max(A=@n22, B=ScalarParameter() {defaultValue=1, group=Icon, parameterName=Icon Opacity})
  n24 = NamedRerouteUsage()
  n25 = NamedRerouteUsage()
  n26 = LinearInterpolate(
   A=StaticSwitchParameter(True=LinearInterpolate(A=@n22, B=@n23, Alpha=@n24), False=@n22) {defaultValue=true, group=Icon, parameterName=Use Icon, sortPriority=1}
   B=@n25
   Alpha=@n10
  )
  n27 = NamedRerouteUsage()
  n28 = Constant()
  MaterialFunctionCall(MA In=MakeMaterialAttributes!(
    BaseColor
    Metallic
    Specular
    Roughness
    Anisotropy
    EmissiveColor=LinearInterpolate(A=@n8, B=@n9, Alpha=@n10)
    Opacity=@n26
    OpacityMask=@n26
    Normal
    Tangent
    WorldPositionOffset=StaticSwitchParameter(True=@n27, False=@n28) {defaultValue=true, parameterName=IsHologram, sortPriority=0}
    SubsurfaceColor
    ClearCoat
    ClearCoatRoughness
    AmbientOcclusion
    Refraction
    CustomizedUVs_0
    CustomizedUVs_1
    CustomizedUVs_2
    CustomizedUVs_3
    CustomizedUVs_4
    CustomizedUVs_5
    CustomizedUVs_6
    CustomizedUVs_7
    PixelDepthOffset
    ShadingModel
    Displacement
   )) {materialFunction=&MF_CreativeMoveTool}
  Multiply(A=@n25, B=NamedRerouteUsage())
  Multiply(B=SphereMask(
    A=PixelDepth().R
    B=Constant()
    Radius=ScalarParameter() {defaultValue=100, group=Distance Fade, parameterName=FadeDistance}
    Hardness=Constant()
   ))
  n29 = MaterialFunctionCall() {materialFunction=&LinearGradient}
  n30 = Abs(None=Subtract(B=@n29.UGradient) {constA=0.5})
  n31 = Max(A=@n30, B=Abs(None=Subtract(B=@n29.VGradient) {constA=0.5}))
  n32 = If(A=@n30, "A > B"=Constant(), "A < B"=Fmod(A=Subtract(A=@n29.VGradient) {constA=0.5, constB=0.48}, B=Constant() {r=0.5})) {constB=0.3}
  n33 = Constant()
  NamedRerouteDeclaration(None=If(A=Max(A=@n31, B=@n32), "A > B"=Constant() {r=1}, "A == B"=@n33, "A < B"=@n33) {constB=0.49}) {name=IconTop, nodeColor=<rgba 0.1719 1 0 1>}
  n34 = Constant() {r=1}
  n35 = If(A=@n34, B=ComponentMask(None=TextureCoordinate() {coordinateIndex=3}) {r=true}, "A > B"=@n34, "A == B"=@n34, "A < B"=Constant()) {desc=Z Mask}
  n36 = Constant() {r=1}
  n37 = ComponentMask(None=TextureCoordinate() {coordinateIndex=1}) {r=true}
  n38 = Constant()
  n39 = MaterialFunctionCall(Float2=TextureCoordinate()) {materialFunction=&BreakOutFloat2Components}
  n40 = Divide(B=ScalarParameter() {defaultValue=1, group=Grid, parameterName=Height, sortPriority=3})
  n41 = Constant()
  n42 = NamedRerouteUsage()
  n43 = Constant() {r=1}
  n44 = TextureCoordinate() {coordinateIndex=2}
  n45 = TextureCoordinate()
  n46 = ScalarParameter() {defaultValue=1, group=Grid, parameterName=Height, sortPriority=3}
  n47 = Constant()
  n48 = LinearInterpolate(
   A=LinearInterpolate(
    A=Constant()
    B=If(A=@n39.R, B=@n40, "A > B"=@n41, "A == B"=@n41, "A < B"=@n42)
    Alpha=If(A=@n36, B=@n37, "A > B"=@n36, "A == B"=@n36, "A < B"=@n38) {desc=Y mask}
   )
   B=If(
    A=MaterialFunctionCall(Float2=@n45) {materialFunction=&BreakOutFloat2Components}.R
    B=Divide(B=@n46)
    "A > B"=@n47
    "A == B"=@n47
    "A < B"=NamedRerouteUsage()
   )
   Alpha=If(A=@n43, B=ComponentMask(None=@n44) {r=true}, "A > B"=@n43, "A == B"=@n43, "A < B"=Constant()) {desc=X Mask}
  )
  n49 = ScalarParameter() {group=Icon, parameterName=Icon Inside Opacity}
  n50 = Saturate(None=TwoSidedSign())
  NamedRerouteDeclaration(None=Multiply(A=LinearInterpolate(A=@n48, B=@n2, Alpha=@n35), B=LinearInterpolate(A=@n49, Alpha=@n50) {desc=Visbility Reduction from Inside})) {name=Icon, nodeColor=<rgba 1 0 0.2812 1>}
  n51 = ParticleColor()
  n52 = ScalarParameter() {parameterName=UseParticleColor}
  n53 = Divide(A=ObjectRadius(), B=ScalarParameter() {defaultValue=0.1, group=Noise, parameterName=NoiseSize})
  n54 = Multiply(A=Time(), B=VectorParameter() {group=Noise, parameterName=NoisePanSpeed, defaultValue=<rgba -0.08 0.07 -0.2 0>}.RGB)
  n55 = MaterialFunctionCall(
   TextureObject=TextureObject() {samplerType=SAMPLERTYPE_LinearColor, texture=&T_NoisePack_LinearColor}
   TextureSize=Multiply(A=VectorParameter() {group=Noise, parameterName=NoiseTileMult, defaultValue=<rgba 1 1 0.1 1>}.RGB, B=@n53)
   WorldPosition=Add(A=WorldPosition().XYZ, B=Multiply(A=@n53, B=@n54))
  ) {materialFunction=&WorldAlignedTexture}
  n56 = ActorPositionWS()
  n57 = MaterialFunctionCall(Float2=Round(None=Divide(A=@n56) {constB=2000})) {materialFunction=&BreakOutFloat2Components}
  n58 = Add(A=Time(), B=Add(A=@n57.R, B=@n57.R))
  n59 = Multiply(A=ScalarParameter() {defaultValue=0.5, parameterName=FlickerSpeed}, B=@n58)
  n60 = AppendVector(A=Sine(None=@n59) {period=1.6}, B=Sine(None=@n59) {period=1.42})
  n61 = Sine(None=@n59) {period=1.51}
  n62 = Sine(None=@n59) {period=2.1}
  n63 = ScalarParameter() {defaultValue=0.5, group=Noise, parameterName=NoiseFrequency}
  n64 = Multiply(B=@n63) {constA=0.25}
  n65 = Add(A=Sine(None=@n59) {period=0.72}, B=Multiply(B=@n63) {constA=0.75}) {constB=0.75}
  n66 = Add(A=Multiply(B=@n63) {constA=0.9, constB=0.9}, B=Sine(None=@n59) {period=1.25}) {constB=0.9}
  n67 = Multiply(A=Multiply(A=Add(A=@n62, B=@n64), B=@n65), B=Multiply(A=@n65, B=Ceil(None=@n66)))
  n68 = ConstantBiasScale(None=Sine(None=@n58) {period=0.075})
  n69 = Multiply(A=Ceil(None=AppendVector(A=@n60, B=@n61)), B=Multiply(A=Saturate(None=@n67), B=LinearInterpolate(Alpha=@n68) {constA=0.7}))
  n70 = ScalarParameter() {defaultValue=2, group=Noise, parameterName=NoiseMult}
  n71 = Multiply(
   A=Multiply(A=Power(Base=@n55.XY Texture) {constExponent=4}, B=ScalarParameter() {defaultValue=2, group=Noise, parameterName=NoiseMaskMult}) {constB=20}
   B=Add(A=ScalarParameter() {group=Noise, parameterName=NoiseAdd}, B=Multiply(A=@n69, B=@n70))
  )
  n72 = WorldPosition()
  n73 = ComponentMask(None=MaterialFunctionCall() {materialFunction=&ObjectPivotPoint}.Object Pivot Location) {b=true}
  n74 = Add(A=@n73, B=ScalarParameter() {defaultValue=300, parameterName=PulseHeight}) {constB=1600}
  n75 = ScalarParameter() {defaultValue=1, parameterName=TurnOn}
  n76 = ScalarParameter() {defaultValue=33, parameterName=PulseRadius}
  n77 = Multiply(A=@n76) {constB=0.1}
  n78 = Multiply(
   A=SphereMask(
    A=ComponentMask(None=@n72.XYZ) {b=true}
    B=LinearInterpolate(A=@n73, B=@n74, Alpha=@n75) {constA=-0.25, constB=1.2}
    Radius=LinearInterpolate(A=@n76, B=@n77, Alpha=@n75) {constA=500, constB=100}
   ) {hardnessPercent=0}
   B=OneMinus(None=@n75)
  )
  n79 = Add(
   A=Desaturation(None=Saturate(None=@n71))
   B=Multiply(A=NamedRerouteDeclaration(None=@n78) {name=Hologram Pulse, nodeColor=<rgba 0.4297 0 1 1>}) {constB=2}
  )
  n80 = QualitySwitch(
   Default=MaterialFunctionCall(In=@n79, Contrast=ScalarParameter() {defaultValue=0.2, group=Noise, parameterName=NoiseContrast}) {materialFunction=&CheapContrast}.Result
   Low=Constant()
  )
  n81 = LinearInterpolate(B=ScalarParameter() {defaultValue=0.6, group=Noise, parameterName=NoiseFresnelXPMult}, Alpha=@n80) {constA=1, constB=0.5}
  n82 = PixelNormalWS()
  n83 = Fresnel(
   ExponentIn=Multiply(A=ScalarParameter() {defaultValue=1, parameterName=FresnelExp}, B=@n81)
   BaseReflectFractionIn=ScalarParameter() {defaultValue=0.5, parameterName=FresnelMin}
   Normal=@n82
  ) {baseReflectFraction=0, exponent=0.5}
  n84 = Multiply(A=@n81, B=ScalarParameter() {defaultValue=1, parameterName=EdgeFresnelXP})
  n85 = Saturate(None=LinearInterpolate(Alpha=Fresnel(ExponentIn=@n84, Normal=@n82) {baseReflectFraction=0, exponent=1}) {constA=5, constB=-0.4})
  n86 = Saturate(None=@n80)
  n87 = Add(
   A=Power(Base=Multiply(A=@n83, B=@n85))
   B=Multiply(A=ScalarParameter() {defaultValue=-0.4, group=Noise, parameterName=NoiseBrightness}, B=@n86) {constB=0.1}
  )
  n88 = Multiply(A=Constant() {r=60}, B=LinearInterpolate(Alpha=Ceil(None=@n86)) {constA=1, constB=1.25})
  n89 = ScreenPosition()
  n90 = MaterialFunctionCall(Float2=ViewSize()) {materialFunction=&BreakOutFloat2Components}
  n91 = AppendVector(A=Constant() {r=1}, B=Divide(A=@n90.G, B=@n90.G))
  n92 = ConstantBiasScale(None=Multiply(A=ConstantBiasScale(None=@n89.ViewportUV) {bias=-0.5, scale=2}, B=@n91) {desc=Keep 1:1 aspect ratio regardless of screen aspect ratio}) {scale=2}
  n93 = Constant2Vector() {g=-1, r=1}
  n94 = MaterialFunctionCall(World Position=ActorPositionWS() {desc=Also use Object Position}) {materialFunction=&TransformToClipSpace}
  n95 = Multiply(A=Multiply(A=@n91, B=Multiply(A=@n93, B=@n94.Clip Space XY))) {constB=0.5, desc=Halve offsets to account for 0...1 UV space}
  n96 = Add(
   A=MaterialFunctionCall(UVs=Subtract(A=@n92, B=@n95), "Rotation Angle (0-1)"=Constant() {r=0.25}) {materialFunction=&CustomRotator}.Rotated Values
   B=Multiply(A=Time(), B=Constant() {r=-0.02})
  )
  n97 = ConstantBiasScale(None=Sine(None=Multiply(A=@n88, B=@n96) {constA=1}))
  n98 = Saturate(None=Add(A=Constant() {r=0.5}, B=@n97))
  n99 = WorldPosition()
  n100 = ObjectPositionWS()
  n101 = Divide(A=ComponentMask(None=Subtract(A=@n99.XYZ, B=@n100)) {b=true}, B=ObjectRadius())
  n102 = MaterialFunctionCall(In=Saturate(None=ConstantBiasScale(None=@n101)), Contrast=ScalarParameter() {group=VerticalGradient, parameterName=VerticalGradientContrast}) {materialFunction=&CheapContrast}
  n103 = OneMinus(None=@n102.Result)
  n104 = ScalarParameter() {group=VerticalGradient, parameterName=FlipGradient}
  n105 = Add(
   A=LinearInterpolate(
    A=Multiply(A=@n87, B=@n98)
    B=LinearInterpolate(A=@n102.Result, B=@n103, Alpha=@n104)
    Alpha=ScalarParameter() {defaultValue=0.4, group=VerticalGradient, parameterName=VerticalGradientInflience}
   )
   B=ScalarParameter() {parameterName=Fill}
  )
  n106 = Add(A=Multiply(A=ScalarParameter() {defaultValue=1, parameterName=TurnOn}, B=Clamp(None=@n105)), B=NamedRerouteUsage()) {constA=1}
  n107 = DepthFade(Opacity=@n106, FadeDistance=ScalarParameter() {defaultValue=100, group=Distance Fade, parameterName=DepthFade})
  n108 = Constant() {r=0.6}
  n109 = Multiply(A=@n80) {constB=1.5}
  n110 = LinearInterpolate(
   A=Clamp(None=Subtract(A=@n108, B=@n109)) {minDefault=-0.25}
   B=Add(A=@n109)
   Alpha=MaterialFunctionCall(Float2=Power(Base=@n97) {constExponent=1.5}) {materialFunction=&BreakOutFloat2Components}.R
  ) {constA=0.5}
  n111 = Multiply(A=Saturate(None=Multiply(A=@n107, B=@n110)), B=ScalarParameter() {defaultValue=1.25, parameterName=OpacityMult})
  n112 = PixelDepth()
  n113 = ScalarParameter() {defaultValue=-0.25, parameterName=BiasSorting_Far}
  n114 = SceneTexture() {sceneTextureId=PPI_CustomDepth}
  n115 = Constant()
  n116 = ShadingPathSwitch(
   Default=If(A=Add(A=@n112.R, B=@n113), B=ComponentMask(None=@n114.Color) {r=true}, "A > B"=@n115, "A == B"=@n115, "A < B"=Constant() {r=1})
   Mobile=Constant() {r=1}
  ) {desc=Migrated from FeatureLevelSwitch}
  NamedRerouteDeclaration(None=Multiply(A=LinearInterpolate(B=@n51.A, Alpha=@n52) {constA=1, desc=Workaround for skel mesh bug}, B=Multiply(A=@n111, B=@n116))) {name=Hologram - Opacity, nodeColor=<rgba 0 0.9062 1 1>}
  n117 = ComponentMask(None=Transform(None=ReflectionVectorWS()) {transformSourceType=TRANSFORMSOURCE_World, transformType=TRANSFORM_View}) {g=true, r=true}
  n118 = ScalarParameter() {defaultValue=1, group=Reflection, parameterName=ReflectionMapTilingU, sortPriority=5}
  n119 = ScalarParameter() {defaultValue=1, group=Reflection, parameterName=ReflectionMapTilingV, sortPriority=5}
  n120 = TextureSampleParameter2D(UVs=Multiply(A=ConstantBiasScale(None=@n117), B=AppendVector(A=@n118, B=@n119))) {group=Reflection, parameterName=SphereReflectionMap, sortPriority=1, texture=/Game/Packages/Fortress_SharedMaterials/Textures/ReflectionMap2}
  n121 = Power(Base=@n120.RGB, Exp=ScalarParameter() {defaultValue=1, group=Reflection, parameterName=ReflectionExponent})
  n122 = VectorParameter() {group=Reflection, parameterName=Reflection Tint, sortPriority=2, defaultValue=<rgba 1 1 1 1>}
  n123 = CollectionParameter() {
   collection=&FortniteMaterialParameters
   parameterName=SunAndMoonModelDirectionalVector
  }
  n124 = Clamp(None=DotProduct(A=ComponentMask(None=@n123) {b=true, g=true, r=true}, B=ReflectionVectorWS()))
  n125 = ScalarParameter() {defaultValue=200, group=Specular, parameterName=Spec Highlight Power}
  n126 = Add(
   A=Multiply(A=Multiply(A=@n121, B=@n122.RGB), B=ScalarParameter() {defaultValue=1, group=Reflection, parameterName=Reflection Multiplier})
   B=Multiply(A=Power(Base=@n124, Exp=@n125) {constExponent=20}, B=ScalarParameter() {defaultValue=2, group=Specular, parameterName=Spec Intense})
  )
  n127 = Fresnel(ExponentIn=ScalarParameter() {defaultValue=1, group=Fresnel, parameterName=General FresnelExp}, Normal=VertexNormalWS())
  n128 = ScalarParameter() {defaultValue=0.5, group=Fresnel, parameterName=General FresnelContrast}
  n129 = Multiply(
   A=Multiply(
    A=MaterialFunctionCall(In=@n127, Contrast=@n128) {materialFunction=&CheapContrast}.Result
    B=VectorParameter() {group=Fresnel, parameterName=General FresnelColor, defaultValue=<rgba 0.3916 0.7991 1 1>}.RGB
   )
   B=ScalarParameter() {group=Fresnel, parameterName=General FresnelIntensity}
  )
  n130 = ScalarParameter() {defaultValue=2, parameterName=EmissiveMult}
  n131 = VertexNormalWS()
  n132 = Transform(None=TextureSampleParameter2D() {group=Texture, parameterName=Normal, samplerType=SAMPLERTYPE_Normal, texture=/Engine/EngineMaterials/BaseFlattenNormalMap}.RGB)
  n133 = ScalarParameter() {group=Lighting, parameterName=UseNormal}
  n134 = Clamp(None=DotProduct(A=VectorParameter() {parameterName=LightDirection, defaultValue=<rgba 0 0 1 1>}.RGB, B=LinearInterpolate(A=@n131, B=@n132, Alpha=@n133)))
  n135 = ScalarParameter() {defaultValue=0.2, group=Lighting, parameterName=NormalBrightnessAdd}
  n136 = Saturate(None=Add(A=Multiply(A=@n134, B=@n135) {constB=0.15}, B=@n106))
  n137 = CurveAtlasRowParameter(CurveTime=MaterialFunctionCall(Float2=@n136) {materialFunction=&BreakOutFloat2Components}.R) {
   atlas=&AthenaCurveAtlas01
   curve=/Game/Creative/Devices/Generic/LinearColorCurves/Curve_CP_Device_Hologram_01
   defaultValue=282
   desc=Curve Atlas
   group=Color
   parameterName=CurveAtlas
  }
  n138 = TextureSampleParameter2D() {
   group="Diffuse Texture Overlay (Hologram)"
   parameterName=Diffuse Texture Overlay
   texture=/Game/Creative/Devices/Generic/Materials/Textures/T_CP_Device_Volume_Hologram_NoLogo_01_D
  }
  n139 = ScalarParameter() {group="Diffuse Texture Overlay (Hologram)", parameterName=Diffuse Contribution Cheap Contrast}
  n140 = Desaturation(
   None=LinearInterpolate(
    A=@n138.RGB
    B=MaterialFunctionCall(In=@n138.RGB, Contrast=@n139) {materialFunction=&CheapContrast}.Result
    Alpha=ScalarParameter() {group="Diffuse Texture Overlay (Hologram)", parameterName=Diffuse Overlay Use Cheap Contrast}
   )
   Fraction=ScalarParameter() {defaultValue=1, group="Diffuse Texture Overlay (Hologram)", parameterName=Diffuse Contribution Desaturation}
  )
  n141 = StaticSwitchParameter(
   True=LinearInterpolate(A=@n137, B=Add(A=@n140, B=@n137), Alpha=ScalarParameter() {group="Diffuse Texture Overlay (Hologram)", parameterName=Lerp Diffuse Overlay})
   False=StaticSwitchParameter(True=@n137, False=@n136) {defaultValue=true, group=Color, parameterName=UseColorCurve, sortPriority=0}
  ) {
   defaultValue=true
   desc=Added for consistency matching with other hologram volumes
   group="Diffuse Texture Overlay (Hologram)"
   parameterName=UseDiffuseTextureOverlay
   sortPriority=0
  }
  NamedRerouteDeclaration(None=Add(A=Add(A=@n126, B=@n129), B=Multiply(A=@n141, B=@n130) {constB=2})) {name=Hologram - Color, nodeColor=<rgba 0 1 0.9531 1>}
  n142 = Constant() {r=1}
  n143 = ComponentMask(None=TextureCoordinate() {coordinateIndex=2}) {r=true}
  n144 = TextureCoordinate() {coordinateIndex=2}
  n145 = AppendVector(
   A=Divide(
    A=Constant() {r=1}
    B=MaterialFunctionCall() {materialFunction=&ObjectScale}.Scale Y
   )
   B=Divide(
    A=Constant() {desc="0.85 is default, not 1", r=2}
    B=MaterialFunctionCall() {materialFunction=&ObjectScale}.Scale Z
   )
  )
  n146 = Constant() {r=1}
  n147 = Subtract(A=Add(A=Divide(A=@n144, B=@n145), B=@n146), B=Divide(A=@n146, B=@n145) {desc=Scale the center offset by the scaling factor}) {desc=Re offset to maintain center}
  n148 = Constant()
  n149 = Constant() {r=1}
  n150 = ComponentMask(None=TextureCoordinate() {coordinateIndex=1}) {r=true}
  n151 = TextureCoordinate() {coordinateIndex=1}
  n152 = AppendVector(
   A=Divide(
    A=Constant() {r=1}
    B=MaterialFunctionCall() {materialFunction=&ObjectScale}.Scale X
   )
   B=Divide(
    A=Constant() {desc="0.85 is default, not 1", r=2}
    B=MaterialFunctionCall() {materialFunction=&ObjectScale}.Scale Z
   )
  )
  n153 = Constant() {r=1}
  n154 = Subtract(A=Add(A=Divide(A=@n151, B=@n152), B=@n153), B=Divide(A=@n153, B=@n152) {desc=Scale the center offset by the scaling factor}) {desc=Re offset to maintain center}
  n155 = Constant()
  n156 = TextureSample(UVs=Add(A=If(A=@n142, B=@n143, "A > B"=@n147, "A < B"=@n148) {desc=0}, B=If(A=@n149, B=@n150, "A > B"=@n154, "A < B"=@n155) {desc=0})) {
   samplerSource=SSM_Clamp_WorldGroupSettings
   texture=/Game/Creative/Devices/DeviceZone_Base/Materials/Textures/T_CP_DeviceZone_Bounds_01_BarrierPlacer_D
  }
  n157 = Multiply(A=@n156.G, B=ScalarParameter() {defaultValue=0.4, group=BG, parameterName=BottomGradientOpacity}) {constB=0.5}
  NamedRerouteDeclaration(None=@n157) {name=Bottom Gradient, nodeColor=<rgba 1 0.2812 0 1>}
  n158 = CameraPositionWS()
  n159 = WorldPosition()
  n160 = Divide(
   A=Add(A=Distance(A=@n158, B=@n159.XYZ), B=ScalarParameter() {defaultValue=-100, group=Distance Fade, parameterName=DistanceBias})
   B=ScalarParameter() {defaultValue=1200, group=Distance Fade, parameterName=FadeScale}
  )
  NamedRerouteDeclaration(None=OneMinus(None=Saturate(None=@n160))) {name=Dist Fade Multiplier, nodeColor=<rgba 1 0 0.0937 1>}
  n161 = Normalize(VectorInput=Subtract(A=WorldPosition().XYZ, B=ObjectPositionWS()))
  n162 = Constant3Vector() {constant=<rgba 1 1 0 0>}
  n163 = ScalarParameter() {defaultValue=10, group=Noise, parameterName=WPOStrength}
  n164 = Multiply(A=ObjectRadius(), B=ScalarParameter() {defaultValue=0.025, group=Noise, parameterName=WPOStrengthFromObjectRadius}) {constB=0.05}
  n165 = Multiply(A=@n79, B=Multiply(A=Multiply(A=@n161, B=@n162), B=Add(A=@n163, B=@n164))) {constB=0}
  NamedRerouteDeclaration(None=@n165) {name=Hologram WPO, nodeColor=<rgba 0 0.4844 1 1>}
  n166 = Multiply(
   A=MaterialFunctionCall() {materialFunction=&LinearGradient}.UGradient
   B=ScalarParameter() {defaultValue=1, group=Grid, parameterName=Height, sortPriority=3}
  )
  n167 = ScalarParameter() {defaultValue=1, group=Grid, parameterName=Width, sortPriority=1}
  n168 = MaterialFunctionCall(Float2=MaterialFunctionCall(UVs=TextureCoordinate(), Texture Scale=Divide(B=@n167)) {materialFunction=&ScaleUVsByCenter}.UVs) {desc=R doesn't work for some reason..., materialFunction=&BreakOutFloat2Components}
  NamedRerouteDeclaration(None=TextureSample(UVs=AppendVector(A=@n166, B=@n168.G) {desc=Height amd Width}) {desc=Front and Back, texture=/Game/Creative/Devices/Map/Textures/T_CP_Device_Map_Arrow_01_D}.RGBA) {name=IconFrontBack, nodeColor=<rgba 0 1 0.5078 1>}
  n169 = ScalarParameter() {defaultValue=1, group=Grid, parameterName=Depth, sortPriority=2}
  n170 = MaterialFunctionCall(Float2=MaterialFunctionCall(UVs=TextureCoordinate(), Texture Scale=Divide(B=@n169)) {materialFunction=&ScaleUVsByCenter}.UVs) {desc=R doesn't work for some reason..., materialFunction=&BreakOutFloat2Components}
  NamedRerouteDeclaration(None=TextureSample(UVs=AppendVector(A=@n166, B=@n170.G) {desc=Height and Depth}) {desc=Sides, texture=&T_CP_Device_Map_Arrow_01_D}.RGBA) {name=IconSides, nodeColor=<rgba 1 0.6563 0 1>}
  n171 = ScalarParameter() {defaultValue=0.1, group=BG, parameterName=BG Color Opacity, sortPriority=3}
  n172 = LinearInterpolate(A=@n171, Alpha=Saturate(None=TwoSidedSign())) {constA=1, constB=0}
  n173 = Constant()
  n174 = Constant() {r=1}
  n175 = LinearInterpolate(
   A=ScalarParameter() {defaultValue=0.1, group=BLOCKS, parameterName=BlocksBackfaceOpacity, sortPriority=11}
   B=LinearInterpolate(
    A=StaticSwitchParameter(True=@n173, False=@n174) {desc=Does not work with Hologram, parameterName=Fill Backface Only}
    B=@n173
    Alpha=ScalarParameter() {group=Level Loader, parameterName=Is Loaded, sliderMax=1, sortPriority=10}
   ) {constA=1, constB=0, desc=For Level Loader only. Bypasses Fill Backface in blocks opacity on Hologram version.}
   Alpha=Saturate(None=TwoSidedSign())
  )
  n176 = MaterialFunctionCall(Float3=Divide(A=WorldPosition().XYZ, B=ScalarParameter() {defaultValue=450, parameterName=WorldScale})) {materialFunction=&BreakOutFloat3Components}
  n177 = VertexNormalWS()
  n178 = MaterialFunctionCall(Float3=Abs(None=Round(None=@n177))) {materialFunction=&BreakOutFloat3Components}
  n179 = LinearInterpolate(
   A=LinearInterpolate(A=AppendVector(A=@n176.G, B=@n176.G), B=AppendVector(A=@n176.R, B=@n176.R), Alpha=@n178.G)
   B=AppendVector(A=@n176.R, B=@n176.R)
   Alpha=@n178.B
  )
  n180 = Constant() {r=1.5}
  n181 = Constant()
  n182 = Constant() {r=0.25}
  n183 = Multiply(A=Time(), B=Constant() {r=0.53})
  n184 = Panner(Coordinate=Add(A=Multiply(A=@n179, B=@n180) {constB=0.25}, B=AppendVector(A=@n181, B=@n182)), Time=Ceil(None=@n183)) {bFractionalPart=true, speedX=-0.25, speedY=-0.25}
  n185 = Multiply(A=Desaturation(None=TextureSample(UVs=@n184) {samplerType=SAMPLERTYPE_LinearColor, texture=/Game/Packages/Fortress_Sky/Materials/Noise32x322_nearestNeighboor}.RGB)) {constB=10}
  n186 = Saturate(None=Ceil(None=OneMinus(None=@n185)))
  n187 = StaticBool()
  n188 = Multiply(
   A=@n186
   B=MaterialFunctionCall(Value=@n183, Sine Phase=@n187) {materialFunction=&LinearSine}."Rounded Linear Sine "
  )
  n189 = Multiply(A=@n179, B=Constant() {r=1.5}) {constB=0.25}
  n190 = AppendVector(A=Constant() {r=0.5}, B=Constant())
  n191 = Multiply(A=Time(), B=Constant() {r=0.96})
  n192 = Panner(Coordinate=Multiply(A=Add(A=@n189, B=@n190), B=Constant() {r=1}) {constB=0.25}, Time=Ceil(None=@n191)) {bFractionalPart=true, speedX=-0.25, speedY=-0.25}
  n193 = Multiply(A=Desaturation(None=TextureSample(UVs=@n192) {samplerType=SAMPLERTYPE_LinearColor, texture=&Noise32x322_nearestNeighboor}.RGB)) {constB=10}
  n194 = Saturate(None=Ceil(None=OneMinus(None=@n193)))
  n195 = Multiply(
   A=@n194
   B=MaterialFunctionCall(Value=@n191, Sine Phase=@n187) {materialFunction=&LinearSine}."Rounded Linear Sine "
  )
  n196 = Multiply(A=@n179, B=Constant() {r=1.5}) {constB=0.25}
  n197 = AppendVector(A=Constant() {r=0.25}, B=Constant())
  n198 = Multiply(A=Time(), B=Constant() {r=1.43})
  n199 = Panner(Coordinate=Multiply(A=Add(A=@n196, B=@n197), B=Constant() {r=1}) {constB=0.25}, Time=Ceil(None=@n198)) {bFractionalPart=true, speedX=-0.25, speedY=-0.25}
  n200 = Multiply(A=Desaturation(None=TextureSample(UVs=@n199) {samplerType=SAMPLERTYPE_LinearColor, texture=&Noise32x322_nearestNeighboor}.RGB)) {constB=10}
  n201 = Saturate(None=Ceil(None=OneMinus(None=@n200)))
  n202 = MaterialFunctionCall(Value=@n198, Sine Phase=@n187) {materialFunction=&LinearSine}
  n203 = Constant() {r=1.5}
  n204 = Constant()
  n205 = Constant() {r=0.5}
  n206 = Multiply(A=Time(), B=Constant() {r=1.94})
  n207 = Panner(Coordinate=Add(A=Multiply(A=@n179, B=@n203) {constB=0.25}, B=AppendVector(A=@n204, B=@n205)), Time=Ceil(None=@n206)) {bFractionalPart=true, speedX=-0.25, speedY=-0.25}
  n208 = Multiply(A=Desaturation(None=TextureSample(UVs=@n207) {samplerType=SAMPLERTYPE_LinearColor, texture=&Noise32x322_nearestNeighboor}.RGB)) {constB=10}
  n209 = Saturate(None=Ceil(None=OneMinus(None=@n208)))
  n210 = Add(
   A=Add(A=Add(A=@n188, B=@n195), B=Multiply(A=@n201, B=@n202."Rounded Linear Sine "))
   B=Multiply(
    A=@n209
    B=MaterialFunctionCall(Value=@n206, Sine Phase=@n187) {materialFunction=&LinearSine}."Rounded Linear Sine "
   )
  )
  n211 = Multiply(A=@n210, B=@n175)
  n212 = LinearInterpolate(
   A=LinearInterpolate(
    A=StaticSwitchParameter(True=@n172, False=@n171) {desc=Does not work with Hologram, parameterName=Fill Backface Only}
    B=@n172
    Alpha=ScalarParameter() {group=Level Loader, parameterName=Is Loaded, sliderMax=1, sortPriority=10}
   ) {constA=1, constB=0, desc=Make Double-sided}
   B=ScalarParameter() {defaultValue=0.2, group=BLOCKS, parameterName=Blocks Opacity, sortPriority=10}
   Alpha=LinearInterpolate(
    A=StaticSwitchParameter(True=@n210, False=@n211) {desc=Does not work with Hologram, parameterName=Fill Backface Only}
    B=@n210
    Alpha=ScalarParameter() {group=Level Loader, parameterName=Is Loaded, sliderMax=1, sortPriority=10}
   ) {constA=1, constB=0, desc=For Level Loader only. Bypasses Fill Backface in blocks opacity on Hologram version.}
  )
  n213 = Constant() {r=1}
  n214 = TextureCoordinate() {coordinateIndex=4}
  n215 = Multiply(
   A=If(A=@n213, B=ComponentMask(None=@n214) {r=true}, "A > B"=@n213, "A == B"=@n213, "A < B"=Constant()) {desc=Bottom Only Mask}
   B=ScalarParameter() {defaultValue=0.5, group=Level Loader, parameterName=Bottom Opacity Multiplier}
  ) {constB=0.5}
  NamedRerouteDeclaration(None=LinearInterpolate(A=Multiply(A=@n212, B=@n215), Alpha=NamedRerouteUsage())) {name=Level Loader Glow Only - Opacity, nodeColor=<rgba 0 0.7422 1 1>}
  n216 = TextureCoordinate()
  n217 = ScalarParameter() {defaultValue=1, group=Grid, parameterName=Depth, sortPriority=2}
  n218 = ScalarParameter() {defaultValue=1, group=Grid, parameterName=Width, sortPriority=1}
  n219 = AppendVector(A=@n217, B=@n218)
  n220 = ScalarParameter() {defaultValue=1, group=Grid, parameterName=Height, sortPriority=3}
  n221 = AppendVector(A=@n220, B=@n218)
  n222 = Constant() {r=1}
  n223 = TextureCoordinate() {coordinateIndex=1}
  n224 = LinearInterpolate(
   A=TextureSample(UVs=Multiply(A=@n216, B=@n219)) {desc=Top and Bottom, texture=/Game/Athena/Apollo/Environments/Landscape/Textures/T_Apollo_GridDebug_01}.RGB
   B=TextureSample(UVs=Multiply(A=@n216, B=@n221)) {desc=Front and Back, texture=&T_Apollo_GridDebug_01}.RGB
   Alpha=If(A=@n222, B=ComponentMask(None=@n223) {r=true}, "A > B"=@n222, "A == B"=@n222, "A < B"=Constant()) {desc=Y mask}
  )
  n225 = TextureSample(UVs=Multiply(A=@n216, B=AppendVector(A=@n220, B=@n217))) {desc=Sides, texture=&T_Apollo_GridDebug_01}
  n226 = Constant() {r=1}
  n227 = If(A=@n226, B=ComponentMask(None=TextureCoordinate() {coordinateIndex=2}) {r=true}, "A > B"=@n226, "A == B"=@n226, "A < B"=Constant()) {desc=X Mask}
  n228 = ScalarParameter() {defaultValue=0.5, group=Grid, parameterName=Grid Backface Opacity}
  n229 = Saturate(None=TwoSidedSign())
  NamedRerouteDeclaration(None=Multiply(A=LinearInterpolate(A=@n224, B=@n225.RGB, Alpha=@n227), B=LinearInterpolate(A=@n228, Alpha=@n229) {desc=Visbility Reduction from Inside})) {name=Grid, nodeColor=<rgba 0.875 1 0 1>}
  n230 = ScalarParameter() {defaultValue=0.65, parameterName=FresnelBrightness}
  n231 = ScalarParameter() {defaultValue=4, parameterName=FresnelExponent}
  n232 = Constant()
  n233 = StaticBool() {value=true}
  n234 = Add(
   A=Multiply(
    A=MaterialFunctionCall(Brightness=@n230, Exponent=@n231) {materialFunction=&MF_Fresnel_FX}.Fresnel Inverted Double Sided
    B=MaterialFunctionCall!(Mask ID=@n232, Invert Mask=@n233) {desc=Remove Top Brightening, materialFunction=/Game/Creative/Devices/Generic/MaterialFunctions/MF_CP_Device_VolumeBox_FaceMasks}.Result
   )
   B=@n157
  )
  NamedRerouteDeclaration(None=LinearInterpolate(A=@n212, Alpha=@n234)) {name=Blocks - Opacity, nodeColor=<rgba 1 0.1625 0.2823 1>}
  n235 = VectorParameter() {group=BG, parameterName=BG Color, sortPriority=1, defaultValue=<rgba 0 0.2375 0.0931 1>}
  n236 = VectorParameter() {group=BG, parameterName=BG Bottom Glow Color, sortPriority=2, defaultValue=<rgba 0 1 0 1>}
  n237 = Multiply(
   A=Multiply(
    A=VectorParameter() {group=BLOCKS, parameterName=Color 3, sortPriority=3, defaultValue=<rgba 0.0212 0.5711 0.4397 1>}.RGB
    B=ScalarParameter() {defaultValue=1, group=BLOCKS, parameterName=Color 3 Intensity Multiplier, sortPriority=7}
   )
   B=@n186
  )
  n238 = Multiply(
   A=Multiply(
    A=VectorParameter() {group=BLOCKS, parameterName=Color 2, sortPriority=2, defaultValue=<rgba 0.2307 0.5841 0.4508 1>}.RGB
    B=ScalarParameter() {defaultValue=1, group=BLOCKS, parameterName=Color 2 Intensity Multiplier, sortPriority=6}
   )
   B=@n194
  )
  n239 = Multiply(
   A=VectorParameter() {group=BLOCKS, parameterName=Color 1, sortPriority=1, defaultValue=<rgba 0.0561 0.4793 0.3005 1>}.RGB
   B=ScalarParameter() {defaultValue=1, group=BLOCKS, parameterName=Color 1 Intensity Multipler, sortPriority=5}
  )
  n240 = Multiply(
   A=Multiply(
    A=VectorParameter() {group=BLOCKS, parameterName=Color 4, sortPriority=4, defaultValue=<rgba 0.0299 0.3583 0.2798 1>}.RGB
    B=ScalarParameter() {defaultValue=1, group=BLOCKS, parameterName=Color 4 Intensity Multiplier, sortPriority=8}
   )
   B=@n209
  )
  n241 = LinearInterpolate(A=LinearInterpolate(A=LinearInterpolate(A=@n237, B=@n238, Alpha=@n194), B=Multiply(A=@n239, B=@n201), Alpha=@n201), B=@n240, Alpha=@n209)
  n242 = ScalarParameter() {defaultValue=2, group=BLOCKS, parameterName=Overall Blocks Intensity, sortPriority=9}
  n243 = Add(A=Add(A=Add(A=@n186, B=@n194), B=@n201), B=@n240)
  n244 = StaticSwitchParameter(True=@n243, False=Multiply(A=@n243, B=@n175)) {desc=Does not work with Hologram, parameterName=Fill Backface Only}
  n245 = ScalarParameter() {group=Level Loader, parameterName=Is Loaded, sliderMax=1, sortPriority=10}
  NamedRerouteDeclaration(None=LinearInterpolate(
    A=LinearInterpolate(A=@n235.RGB, B=@n236.RGB, Alpha=@n234)
    B=Multiply(A=@n241, B=@n242) {constB=0.85}
    Alpha=LinearInterpolate(A=@n244, B=@n243, Alpha=@n245) {constA=1, constB=0, desc=For Level Loader only. Bypasses Fill Backface in blocks opacity on Hologram version.}
   )) {name=Blocks - Color, nodeColor=<rgba 1 0.1625 0.2823 1>}
  Reroute(None=@n157)
  Reroute(None=@n78)
  Reroute(None=@n165)
 ]
}
